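# Two-sided unlit translucent material for lightning and electric beam effects, usable on Niagara sprites, ribbons and meshes as well as skeletal meshes.
# A looping time value is squared through Sine and Ceil, flipping the U offset in uneven, accelerating steps to make the arc flicker.
# The slowly panning, V-flippable UVs are center-scaled to sample the ElectroBeam texture.
# Its green channel, multiplied by particle color and alpha, drives adjusted emissive and opacity.
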Material /DealershipSharedLibrary/Assets/Sets/ElectroBeam/MM_Lightning_Electricity
graph {
 properties {
  bUsedWithMeshDeformer=true
  bUsedWithNiagaraMeshParticles=true
  bUsedWithNiagaraRibbons=true
  bUsedWithNiagaraSprites=true
  bUsedWithSkeletalMesh=true
  blendMode=BLEND_Translucent
  pixelDepthOffsetMode=PDOM_Legacy
  shadingModel=MSM_Unlit
  twoSided=true
  usedShadingModels=Unlit
 }
 nodes [
  MaterialExpressionVertexColor_0 = ParticleColor()
  n1 = Multiply(A=Time(), B=ScalarParameter() {defaultValue=1.5, parameterName=OffsetSpeed})
  n2 = ScalarParameter() {defaultValue=0.4, parameterName=OffsetPeriod}
  n3 = Multiply(A=Add(A=Fmod(A=@n1, B=@n2), B=Constant() {r=0.15}), B=Constant() {r=3})
  n4 = Ceil(None=Sine(None=Multiply(A=@n3, B=@n3)))
  n5 = ScalarParameter() {defaultValue=0.5, parameterName=U_OffsetMult}
  n6 = AppendVector(A=Add(A=Multiply(A=@n4, B=@n5), B=ScalarParameter() {parameterName=MirrorOffset}), B=Constant())
  n7 = Constant() {r=1}
  n8 = ScalarParameter() {defaultValue=1, parameterName=FlipV}
  n9 = Panner(Coordinate=Multiply(A=TextureCoordinate(), B=AppendVector(A=@n7, B=@n8))) {speedX=-0.1}
  n10 = Constant() {r=1}
  n11 = ScalarParameter() {defaultValue=1, parameterName=ScaleY}
  n12 = TextureSample(UVs=MaterialFunctionCall(
    UVs=Add(A=@n6, B=@n9)
    Texture Scale=MaterialFunctionCall(X=@n10, Y=@n11) {materialFunction=&MakeFloat2}.Result
   ) {materialFunction=&ScaleUVsByCenter}.UVs) {samplerType=SAMPLERTYPE_LinearColor, texture=&/ElectroBeam_01}
  MaterialExpressionConstantClamp_0 = Saturate(None=Multiply(A=@MaterialExpressionVertexColor_0.A, B=@n12.G))
  MaterialFunctionCall(Emissive=Clamp(None=Multiply(A=@n12.G, B=@MaterialExpressionVertexColor_0.RGB)) {maxDefault=10000}) {materialFunction=&MF_FX_EmissiveAdjustments}
 ]
}
